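; Blueprint interface for the balloon board in the PopALoon quest interactable in Figment Season 9.
; It declares three functions with no implementation: bursting a balloon, filling a balloon at a given fill speed (default 1.0), and killing all balloons.
; Implementing actors supply the behavior, letting callers drive the board without knowing its concrete class.
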(fn BurstBalloon)

(fn FillBalloon
  (params (FillSpeed :type Float%20%28double-precision%29 :default 1.0)))

(fn KillBalloons)
